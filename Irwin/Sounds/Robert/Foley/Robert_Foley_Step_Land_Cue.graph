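# Sound cue for Robert's landing step foley, a creature step sound using heavy predator attenuation and limited concurrency.
# An IsLocal branch plays a boosted random body-fall when true, otherwise deferring to an IsRiddenByEnemy branch.
# That branch mixes body-falls with modulated footstep-land variations when true or unset, and plays only a quieter body-fall when false.
# Random nodes supply the variations and are excluded from branch culling.
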SoundCue /Irwin/Sounds/Robert/Foley/Robert_Foley_Step_Land_Cue
root=Branch_1
nodes [
 {id=Branch_0, children [Mixer_1,Mixer_0,Mixer_1], properties {boolParameterName=IsRiddenByEnemy}}
 {id=Branch_1, children [Mixer_2,Branch_0,Branch_0], properties {boolParameterName=IsLocal}}
 {id=Mixer_0, children [Random_2], properties {inputVolume [0.5]}}
 {id=Mixer_1, children [Random_1,Modulator_1], properties {inputVolume [1.25,1.25]}}
 {id=Mixer_2, children [Random_0,null], properties {inputVolume [1.75,2]}}
 {id=Modulator_1, children [Random_5]}
 {
  id=Random_0
  children [WavePlayer_0,WavePlayer_1,WavePlayer_2,WavePlayer_3,WavePlayer_4,WavePlayer_5,WavePlayer_6,WavePlayer_7]
  properties {bShouldExcludeFromBranchCulling=true, weights [1,1,1,1,1,1,1,1]}
 }
 {
  id=Random_1
  children [WavePlayer_8, WavePlayer_9, WavePlayer_10, WavePlayer_11, WavePlayer_12, WavePlayer_13, WavePlayer_14, WavePlayer_15]
  properties {bShouldExcludeFromBranchCulling=true, weights [1,1,1,1,1,1,1,1]}
 }
 {
  id=Random_2
  children [WavePlayer_16, WavePlayer_17, WavePlayer_18, WavePlayer_19, WavePlayer_20, WavePlayer_21, WavePlayer_22, WavePlayer_23]
  properties {bShouldExcludeFromBranchCulling=true, weights [1,1,1,1,1,1,1,1]}
 }
 {
  id=Random_5
  children [WavePlayer_32, WavePlayer_33, WavePlayer_34, WavePlayer_35]
  properties {bShouldExcludeFromBranchCulling=true, weights [1,1,1,1]}
 }
 {id=WavePlayer_0, properties {soundWaveAssetPtr=&/Robert_Foley_BodyFall_01}}
 {id=WavePlayer_1, properties {soundWaveAssetPtr=&/Robert_Foley_BodyFall_02}}
 {id=WavePlayer_10, properties {soundWaveAssetPtr=&/Robert_Foley_BodyFall_03}}
 {id=WavePlayer_11, properties {soundWaveAssetPtr=&/Robert_Foley_BodyFall_04}}
 {id=WavePlayer_12, properties {soundWaveAssetPtr=&/Robert_Foley_BodyFall_05}}
 {id=WavePlayer_13, properties {soundWaveAssetPtr=&/Robert_Foley_BodyFall_06}}
 {id=WavePlayer_14, properties {soundWaveAssetPtr=&/Robert_Foley_BodyFall_07}}
 {id=WavePlayer_15, properties {soundWaveAssetPtr=&/Robert_Foley_BodyFall_08}}
 {id=WavePlayer_16, properties {soundWaveAssetPtr=&Robert_Foley_BodyFall_01}}
 {id=WavePlayer_17, properties {soundWaveAssetPtr=&Robert_Foley_BodyFall_02}}
 {id=WavePlayer_18, properties {soundWaveAssetPtr=&Robert_Foley_BodyFall_03}}
 {id=WavePlayer_19, properties {soundWaveAssetPtr=&Robert_Foley_BodyFall_04}}
 {id=WavePlayer_2, properties {soundWaveAssetPtr=&Robert_Foley_BodyFall_03}}
 {id=WavePlayer_20, properties {soundWaveAssetPtr=&Robert_Foley_BodyFall_05}}
 {id=WavePlayer_21, properties {soundWaveAssetPtr=&Robert_Foley_BodyFall_06}}
 {id=WavePlayer_22, properties {soundWaveAssetPtr=&Robert_Foley_BodyFall_07}}
 {id=WavePlayer_23, properties {soundWaveAssetPtr=&Robert_Foley_BodyFall_08}}
 {id=WavePlayer_3, properties {soundWaveAssetPtr=&Robert_Foley_BodyFall_04}}
 {id=WavePlayer_32, properties {soundWaveAssetPtr=&/Robert_Foley_Footstep_Land_01}}
 {id=WavePlayer_33, properties {soundWaveAssetPtr=&/Robert_Foley_Footstep_Land_02}}
 {id=WavePlayer_34, properties {soundWaveAssetPtr=&/Robert_Foley_Footstep_Land_03}}
 {id=WavePlayer_35, properties {soundWaveAssetPtr=&/Robert_Foley_Footstep_Land_04}}
 {id=WavePlayer_4, properties {soundWaveAssetPtr=&Robert_Foley_BodyFall_05}}
 {id=WavePlayer_5, properties {soundWaveAssetPtr=&Robert_Foley_BodyFall_06}}
 {id=WavePlayer_6, properties {soundWaveAssetPtr=&Robert_Foley_BodyFall_07}}
 {id=WavePlayer_7, properties {soundWaveAssetPtr=&Robert_Foley_BodyFall_08}}
 {id=WavePlayer_8, properties {soundWaveAssetPtr=&Robert_Foley_BodyFall_01}}
 {id=WavePlayer_9, properties {soundWaveAssetPtr=&Robert_Foley_BodyFall_02}}
]
properties {
 attenuationSettings=/Irwin/Sounds/Settings/Predator_Step_Heavy_Att
 bOverrideConcurrency=true
 maxDistance=6000
 soundClassObject=/Game/Sounds/FortSoundClassesAndModes/Fort_Creatures_Steps_Normal
 virtualizationMode=Disabled
 volumeMultiplier=1
 concurrencyOverrides {maxCount=6}
}
