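# Sound cue for Robert's walking footstep foley, using the creature-steps sound class and heavy predator attenuation.
# IsLocal selects a slightly reduced mix; otherwise IsRiddenByEnemy selects full volume when true or unset, and a much quieter mix when false.
# Each mix randomly draws from the same ten walk waves through its own pitch and volume modulator.
# A concurrency override caps simultaneous steps per owner.
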SoundCue /Irwin/Sounds/Robert/Foley/Robert_Foley_Step_Walk_Cue
root=Branch_1
nodes [
 {id=Branch_0, children [Mixer_1,Mixer_0,Mixer_1], properties {boolParameterName=IsRiddenByEnemy}}
 {id=Branch_1, children [Mixer_2,Branch_0,Branch_0], properties {boolParameterName=IsLocal}}
 {id=Mixer_0, children [Modulator_2], properties {inputVolume [0.3]}}
 {id=Mixer_1, children [Modulator_1], properties {inputVolume [1]}}
 {id=Mixer_2, children [Modulator_0], properties {inputVolume [0.8]}}
 {id=Modulator_0, children [Random_8], properties {pitchMax=1.2, pitchMin=0.9, volumeMax=1.35, volumeMin=0.6}}
 {id=Modulator_1, children [Random_0], properties {pitchMax=1.2, pitchMin=0.9, volumeMin=0.85}}
 {id=Modulator_2, children [Random_1], properties {pitchMax=1.2, pitchMin=0.9, volumeMax=1.25, volumeMin=0.85}}
 {
  id=Random_0
  children [WavePlayer_0,WavePlayer_1,WavePlayer_2,WavePlayer_3,WavePlayer_4,WavePlayer_5,WavePlayer_6,WavePlayer_7,WavePlayer_8,WavePlayer_9]
  properties {bShouldExcludeFromBranchCulling=true, weights [1,1,1,1,1,1,1,1,1,1]}
 }
 {
  id=Random_1
  children [
   WavePlayer_10
   WavePlayer_11
   WavePlayer_12
   WavePlayer_13
   WavePlayer_14
   WavePlayer_15
   WavePlayer_16
   WavePlayer_17
   WavePlayer_18
   WavePlayer_19
  ]
  properties {bShouldExcludeFromBranchCulling=true, weights [1,1,1,1,1,1,1,1,1,1]}
 }
 {
  id=Random_8
  children [
   WavePlayer_31
   WavePlayer_32
   WavePlayer_33
   WavePlayer_34
   WavePlayer_35
   WavePlayer_36
   WavePlayer_37
   WavePlayer_38
   WavePlayer_39
   WavePlayer_40
  ]
  properties {bShouldExcludeFromBranchCulling=true, weights [1,1,1,1,1,1,1,1,1,1]}
 }
 {id=WavePlayer_0, properties {soundWaveAssetPtr=&/Robert_Foley_Footstep_Walk_10}}
 {id=WavePlayer_1, properties {soundWaveAssetPtr=&/Robert_Foley_Footstep_Walk_11}}
 {id=WavePlayer_10, properties {soundWaveAssetPtr=&Robert_Foley_Footstep_Walk_10}}
 {id=WavePlayer_11, properties {soundWaveAssetPtr=&Robert_Foley_Footstep_Walk_11}}
 {id=WavePlayer_12, properties {soundWaveAssetPtr=&/Robert_Foley_Footstep_Walk_12}}
 {id=WavePlayer_13, properties {soundWaveAssetPtr=&/Robert_Foley_Footstep_Walk_13}}
 {id=WavePlayer_14, properties {soundWaveAssetPtr=&/Robert_Foley_Footstep_Walk_14}}
 {id=WavePlayer_15, properties {soundWaveAssetPtr=&/Robert_Foley_Footstep_Walk_15}}
 {id=WavePlayer_16, properties {soundWaveAssetPtr=&/Robert_Foley_Footstep_Walk_01}}
 {id=WavePlayer_17, properties {soundWaveAssetPtr=&/Robert_Foley_Footstep_Walk_02}}
 {id=WavePlayer_18, properties {soundWaveAssetPtr=&/Robert_Foley_Footstep_Walk_03}}
 {id=WavePlayer_19, properties {soundWaveAssetPtr=&/Robert_Foley_Footstep_Walk_04}}
 {id=WavePlayer_2, properties {soundWaveAssetPtr=&Robert_Foley_Footstep_Walk_12}}
 {id=WavePlayer_3, properties {soundWaveAssetPtr=&Robert_Foley_Footstep_Walk_13}}
 {id=WavePlayer_31, properties {soundWaveAssetPtr=&Robert_Foley_Footstep_Walk_10}}
 {id=WavePlayer_32, properties {soundWaveAssetPtr=&Robert_Foley_Footstep_Walk_11}}
 {id=WavePlayer_33, properties {soundWaveAssetPtr=&Robert_Foley_Footstep_Walk_12}}
 {id=WavePlayer_34, properties {soundWaveAssetPtr=&Robert_Foley_Footstep_Walk_13}}
 {id=WavePlayer_35, properties {soundWaveAssetPtr=&Robert_Foley_Footstep_Walk_14}}
 {id=WavePlayer_36, properties {soundWaveAssetPtr=&Robert_Foley_Footstep_Walk_15}}
 {id=WavePlayer_37, properties {soundWaveAssetPtr=&Robert_Foley_Footstep_Walk_01}}
 {id=WavePlayer_38, properties {soundWaveAssetPtr=&Robert_Foley_Footstep_Walk_02}}
 {id=WavePlayer_39, properties {soundWaveAssetPtr=&Robert_Foley_Footstep_Walk_03}}
 {id=WavePlayer_4, properties {soundWaveAssetPtr=&Robert_Foley_Footstep_Walk_14}}
 {id=WavePlayer_40, properties {soundWaveAssetPtr=&Robert_Foley_Footstep_Walk_04}}
 {id=WavePlayer_5, properties {soundWaveAssetPtr=&Robert_Foley_Footstep_Walk_15}}
 {id=WavePlayer_6, properties {soundWaveAssetPtr=&Robert_Foley_Footstep_Walk_01}}
 {id=WavePlayer_7, properties {soundWaveAssetPtr=&Robert_Foley_Footstep_Walk_02}}
 {id=WavePlayer_8, properties {soundWaveAssetPtr=&Robert_Foley_Footstep_Walk_03}}
 {id=WavePlayer_9, properties {soundWaveAssetPtr=&Robert_Foley_Footstep_Walk_04}}
]
properties {
 attenuationSettings=/Irwin/Sounds/Settings/Predator_Step_Heavy_Att
 bOverrideConcurrency=true
 maxDistance=6000
 soundClassObject=/Game/Sounds/FortSoundClassesAndModes/Fort_Creatures_Steps_Normal
 volumeMultiplier=1
 concurrencyOverrides {bLimitToOwner=true, maxCount=2}
}
